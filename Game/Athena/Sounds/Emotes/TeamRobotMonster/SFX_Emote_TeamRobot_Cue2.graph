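# Sound cue for the TeamRobot emote audio in Athena, routed through the first-person emote music sound class.
# A Random node picks one of two equally weighted wave players, playing the TeamRobot 02 or 03 variant.
# It references the first-person emote music concurrency set and carries a PreventNew override setting.
# A submix send feeds the music audio analysis submix.
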SoundCue /Game/Athena/Sounds/Emotes/TeamRobotMonster/SFX_Emote_TeamRobot_Cue2
root=Random_0
nodes [
 {id=Random_0, children [WavePlayer_1,WavePlayer_2], properties {weights [1,1]}}
 {id=WavePlayer_1, properties {soundWaveAssetPtr=&/SFX_Emote_TeamRobot_02}}
 {id=WavePlayer_2, properties {soundWaveAssetPtr=&/SFX_Emote_TeamRobot_03}}
]
properties {
 maxDistance=2097152
 soundClassObject=/Game/Sounds/FortSoundClassesAndModes/Fort_Music_Emote_1P_PSM
 volumeMultiplier=1
 concurrencyOverrides {resolutionRule=PreventNew}
 concurrencySet [/Game/Sounds/Concurrency_Rules/Athena_EmoteMusic_1P_Conc]
 soundSubmixSends [{minSendLevel=0, soundSubmix=&Fortnite_AudioAnalysis_Music}]
}
